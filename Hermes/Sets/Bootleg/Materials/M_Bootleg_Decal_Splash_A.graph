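# M_Bootleg_Decal_Splash_A is a deferred decal material for a splash mark in the Bootleg set.
# It uses translucent blending and the Default Lit shading model.
# The graph holds a single texture sample of the Splash A decal texture, with no connections to material outputs shown and no parameters.
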Material /Hermes/Sets/Bootleg/Materials/M_Bootleg_Decal_Splash_A
graph {
 properties {blendMode=BLEND_Translucent, materialDomain=MD_DeferredDecal, pixelDepthOffsetMode=PDOM_Legacy, usedShadingModels=Default Lit}
 nodes [TextureSample() {texture=/Hermes/Sets/Bootleg/Textures/Decals/T_Bootleg_Decal_Splash_A}]
}
